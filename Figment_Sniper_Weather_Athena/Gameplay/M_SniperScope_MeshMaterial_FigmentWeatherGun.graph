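# Masked material for the Figment weather sniper's scope, showing an emissive storm-tracker map of the island on its screen.
# MPC_Minimap drives current, next and pulsing hot/cold next-next circles, animated storm tint and player dot, with the current circle setting map pan and zoom.
# A screen mask confines the map to the display over dark, noisy-roughness metal.
# AttachScope supplies normal, world position offset and opacity, which is temporally dithered.
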Material /Figment_Sniper_Weather_Athena/Gameplay/M_SniperScope_MeshMaterial_FigmentWeatherGun
graph {
 properties {
  bTangentSpaceNormal=false
  bUseMaterialAttributes=true
  blendMode=BLEND_Masked
  pixelDepthOffsetMode=PDOM_Legacy
  previewMesh=&/S_SniperScope_FigmentWeatherDisplay
  usedShadingModels=Default Lit
 }
 nodes [
  MaterialExpressionConstant4Vector_3 = Constant4Vector() {constant=<rgba 0.5 0.5 1 0>}
  MaterialExpressionCollectionParameter_6 = CollectionParameter() {collection=/Game/Athena/HUD/MiniMap/MPC_Minimap, parameterName=CurrentCircle}
  n1 = Switch!(
   SwitchValue=CollectionParameter() {collection=&MPC_Minimap, parameterName=StormPhaseStarted}
   Default
   PreStorm=@MaterialExpressionConstant4Vector_3.RGBA
   StormActive=@MaterialExpressionCollectionParameter_6
  ) {
   constDefault=1
   inputs [
    {
     inputName=PreStorm
     input {
      expression=&self:MaterialExpressionConstant4Vector_3
      inputName=None
      mask=1
      maskA=1
      maskB=1
      maskG=1
      maskR=1
      outputIndex=0
     }
    }
    {
     inputName=StormActive
     input {
      expression=&self:MaterialExpressionCollectionParameter_6
      inputName=None
      mask=0
      maskA=0
      maskB=0
      maskG=0
      maskR=0
      outputIndex=0
     }
    }
   ]
  }
  n2 = ComponentMask(None=@n1) {b=true}
  n3 = TextureCoordinate()
  Add(B=MaterialFunctionCall(Number=@n2, UVs=@n3) {materialFunction=&DebugScalarValues}.Result)
  n4 = ScalarParameter() {defaultValue=0.125, parameterName=Dots-Tile}
  n5 = If(A=@n2, B=Constant() {r=0.08}, "A > B"=Constant2Vector() {g=2.3, r=0.73}, "A < B"=Constant2Vector() {g=5, r=2.6})
  n6 = Clamp(None=Multiply(A=@n2) {constB=3.33})
  n7 = LinearInterpolate(A=ComponentMask(None=@n5) {g=true}, B=ComponentMask(None=@n5) {r=true}, Alpha=@n6)
  n8 = ComponentMask(None=@n1) {g=true, r=true}
  n9 = MaterialFunctionCall(Float2=Subtract(A=Constant2Vector() {g=0.5, r=0.5}, B=@n8)) {materialFunction=&BreakOutFloat2Components}
  n10 = Multiply(A=@n9.R) {constB=-1}
  n11 = Multiply(A=@n9.G) {constB=-1}
  n12 = Constant() {r=512}
  n13 = MaterialFunctionCall!(
   Texture Scale=AppendVector(A=@n7, B=@n7)
   UVs=Add(A=@n3, B=AppendVector(A=@n10, B=@n11))
   Offset XY Texturesize Z=AppendVector(A=Multiply(A=@n8, B=@n12) {constB=512}, B=@n12)
  ) {materialFunction=/Engine/Functions/Engine_MaterialFunctions01/Texturing/ScaleUVsAroundPoint}
  n14 = Frac(None=Add(A=Constant2Vector() {g=0.5, r=0.5}, B=Divide(A=@n13.UVs, B=@n4) {constA=0, constB=5}))
  n15 = Constant() {r=0.5}
  n16 = ScalarParameter() {defaultValue=0.3, parameterName=Dots-Radius}
  n17 = ScalarParameter() {parameterName=Dots-Hard}
  n18 = TextureSample(UVs=Multiply(A=@n13.UVs) {constB=2}) {samplerSource=SSM_Wrap_WorldGroupSettings, texture=&/T_StormTrackerScreen_AthenaIsland}
  n19 = Add(
   A=Multiply(A=SphereMask(A=@n14, B=@n15, Radius=@n16, Hardness=@n17), B=ScalarParameter() {defaultValue=0.3, parameterName=Dots-Amount})
   B=@n18.G
  )
  n20 = Power(Base=Constant() {r=0.5}, Exp=ScalarParameter() {defaultValue=2, parameterName=GridSweepPower})
  n21 = Multiply(A=@n19, B=@n20)
  n22 = Add(A=@n19, B=@n20)
  n23 = Constant() {r=0.5}
  n24 = Multiply(A=Saturate(None=Add(A=@n22, B=@n23)))
  ComponentMask(None=Multiply(
    A=VectorParameter() {group=EffectParams, parameterName=BackgroundMapColor, defaultValue=<rgba 0.0353 0.298 0.749 1>}.RGB
    B=Add(A=@n21, B=@n24)
   )) {b=true, g=true, r=true}
  Constant() {r=0.125}
  Constant() {r=0.085}
  n25 = Time()
  n26 = ScalarParameter() {defaultValue=1.5, parameterName=SweepSpeed}
  n27 = MaterialFunctionCall(Float2=Rotator(Coordinate=TextureCoordinate() {coordinateIndex=1}, Time=Multiply(A=@n25, B=@n26)) {speed=1}) {materialFunction=&BreakOutFloat2Components}
  n28 = OneMinus(None=@n27.G)
  n29 = OneMinus(None=@n27.R)
  MaterialFunctionCall(Vector or UVs=ConstantBiasScale(None=AppendVector(A=@n28, B=@n29)) {bias=-0.5, scale=2}) {materialFunction=&VectorToRadialValue}
  n30 = ScalarParameter() {defaultValue=3, parameterName=Roughness_U_Scale}
  n31 = ScalarParameter() {defaultValue=3, parameterName=Roughness_V_Scale, sortPriority=1}
  n32 = TextureSampleParameter2D(UVs=Multiply(A=TextureCoordinate(), B=AppendVector(A=@n30, B=@n31))) {parameterName=Roughness Texture, texture=/Engine/MapTemplates/TilingNoise05}
  n33 = AppendVector(A=@n32.RGB, B=@n32.RGB)
  n34 = ShadingPathSwitch(
   Default=LinearInterpolate(
    A=ScalarParameter() {defaultValue=0.4, parameterName=Roughness_Min}
    B=ScalarParameter() {defaultValue=0.8, parameterName=Roughness_Max}
    Alpha=StaticComponentMaskParameter(None=@n33) {defaultR=true, parameterName=Roughness Mask}
   ) {constA=0.4, constB=0.8}
   Mobile=ScalarParameter() {defaultValue=0.5048, parameterName=Roughness_Mobile}
  ) {desc=Migrated from FeatureLevelSwitch}
  n35 = TextureSample() {texture=/Game/Athena/HUD/reticles/T_UI_WeatherScopeMask}
  n36 = OneMinus(None=@n35.R)
  n37 = Multiply(
   A=Constant4Vector() {constant=<rgba 0 0.783 0.8108 0>}.RGBA
   B=TextureSample(UVs=@n13.UVs) {samplerSource=SSM_Clamp_WorldGroupSettings, texture=&T_StormTrackerScreen_AthenaIsland}.R
  )
  n38 = Constant3Vector() {constant=<rgba 0 0.1661 0.6198 1>}
  n39 = Add(A=ComponentMask(None=Multiply(A=@n37) {constB=2}) {b=true, g=true, r=true}, B=Add(A=Multiply(A=@n38, B=@n18.G), B=@n38))
  n40 = VectorParameter() {group=EffectParams, parameterName=TintOutsideSafeZone, defaultValue=<rgba 0.3448 0.015 0.75 0.1>}
  n41 = Panner(Coordinate=Multiply(A=TextureCoordinate() {coordinateIndex=1}, B=ScalarParameter() {defaultValue=1.5, parameterName=Lines-NoiseSize})) {speedY=0.08}
  MaterialExpressionConstant4Vector_1 = Constant4Vector() {constant=<rgba 0.5 0.5 1 0>}
  MaterialExpressionCollectionParameter_0 = CollectionParameter() {collection=&MPC_Minimap, parameterName=CurrentCircle}
  n42 = Switch(
   SwitchValue=CollectionParameter() {collection=&MPC_Minimap, parameterName=StormPhaseStarted}
   PreStorm=@MaterialExpressionConstant4Vector_1.RGBA
   StormActive=@MaterialExpressionCollectionParameter_0
  ) {
   constDefault=1
   inputs [
    {
     inputName=PreStorm
     input {
      expression=&self:MaterialExpressionConstant4Vector_1
      inputName=None
      mask=1
      maskA=1
      maskB=1
      maskG=1
      maskR=1
      outputIndex=0
     }
    }
    {
     inputName=StormActive
     input {
      expression=&self:MaterialExpressionCollectionParameter_0
      inputName=None
      mask=0
      maskA=0
      maskB=0
      maskG=0
      maskR=0
      outputIndex=0
     }
    }
   ]
  }
  n43 = Distance(A=@n13.UVs, B=ComponentMask(None=@n42) {g=true, r=true})
  n44 = Add(
   A=Multiply(
    A=TextureSample(UVs=@n41) {samplerType=SAMPLERTYPE_LinearGrayscale, texture=/Game/Global/Textures/Default/Tileable_Patterns/T_StormCloudDetail01_M}.R
    B=ScalarParameter() {defaultValue=1.5, parameterName=Lines-NoiseAmount}
   )
   B=Multiply(A=ScalarParameter() {defaultValue=9, parameterName=Lines-Tile}, B=@n43)
  )
  n45 = ScalarParameter() {defaultValue=50, parameterName=Lines-PowA}
  n46 = ScalarParameter() {defaultValue=1, parameterName=Lines-PowB}
  n47 = Saturate(None=Multiply(A=ScalarParameter() {defaultValue=1.1, parameterName=Lines-TrailPush}))
  n48 = Add(
   A=Power(
    Base=TextureSample(UVs=@n44) {samplerType=SAMPLERTYPE_LinearColor, texture=/Game/Global/Textures/Default/Masks_and_Gradients/T_VertGradient}.R
    Exp=LinearInterpolate(A=@n45, B=@n46, Alpha=@n47)
   )
   B=ScalarParameter() {defaultValue=0.25, parameterName=Lines-Add}
  )
  n49 = Power(Base=Constant() {r=0.5}, Exp=ScalarParameter() {defaultValue=3, parameterName=CloudsSweepPower})
  n50 = ScalarParameter() {defaultValue=6, parameterName=CloudsSweepBright}
  n51 = Constant() {r=0.001}
  n52 = LinearInterpolate(A=Constant() {r=15}, B=ScalarParameter() {defaultValue=150, parameterName=Current-ScaleFactorOuter}, Alpha=@n6)
  n53 = ComponentMask(None=@n42) {b=true}
  n54 = MaterialFunctionCall(Alpha=@n43, Min=@n53, Max=Add(A=@n53, B=Multiply(A=@n51, B=@n52))) {materialFunction=&SmoothStep}
  n55 = Add(
   A=Multiply(A=Saturate(None=@n48), B=Multiply(A=@n49, B=@n50))
   B=Multiply(A=OneMinus(None=@n54.Result), B=ScalarParameter() {defaultValue=0.25, parameterName=StormEdgeFalloffAmount})
  )
  n56 = Constant() {r=0.001}
  n57 = LinearInterpolate(A=Constant() {r=5}, B=ScalarParameter() {defaultValue=40, parameterName=Current-ScaleFactor}, Alpha=@n6)
  n58 = MaterialFunctionCall(Alpha=@n43, Min=Subtract(A=@n53, B=Multiply(A=@n56, B=@n57)), Max=@n53) {materialFunction=&SmoothStep}
  n59 = ScalarParameter() {defaultValue=0.75, parameterName=IslandInsideStormEffect}
  n60 = Add(A=@n40.A, B=Add(A=Multiply(A=@n55, B=@n58.Result), B=Multiply(A=@n59) {constB=0.25}))
  n61 = CollectionParameter() {collection=&MPC_Minimap, parameterName=NextCircle}
  n62 = Distance(A=@n13.UVs, B=ComponentMask(None=@n61) {g=true, r=true})
  n63 = LinearInterpolate(A=Constant() {r=0.5}, B=ScalarParameter() {defaultValue=20, parameterName=Next-ScaleFactor}, Alpha=@n6)
  n64 = Multiply(A=Constant() {r=0.001}, B=@n63)
  n65 = ComponentMask(None=@n61) {b=true}
  n66 = MaterialFunctionCall(Alpha=@n62, Min=Subtract(A=@n65, B=@n64), Max=@n65) {materialFunction=&SmoothStep}
  n67 = LinearInterpolate(A=@n39, B=Multiply(A=Multiply(A=@n40.RGB, B=@n60), B=@n66.Result), Alpha=@n58.Result)
  n68 = MaterialFunctionCall(Alpha=@n62, Min=@n65, Max=Add(A=@n65, B=@n64)) {materialFunction=&SmoothStep}
  n69 = Multiply(A=Constant3Vector() {constant=<rgba 1 0.9594 0 0>}, B=Subtract(A=Add(A=@n66.Result), B=Add(A=@n68.Result)))
  n70 = LinearInterpolate(
   A=Constant3Vector() {constant=<rgba 0.4566 0.4566 0.4566 0>}
   B=Multiply(A=Constant3Vector() {constant=<rgba 0.9843 0 1 0>}) {constB=4}
   Alpha=CollectionParameter() {collection=&MPC_Minimap, parameterName=NextNextCircleHotOrCold}
  ) {constA=0.5, constB=1.3}
  n71 = CollectionParameter() {collection=&MPC_Minimap, parameterName=NextNextCircle}
  n72 = Distance(A=@n13.UVs, B=ComponentMask(None=@n71) {g=true, r=true})
  n73 = Multiply(A=Constant() {r=0.001}, B=@n63)
  n74 = ComponentMask(None=@n71) {b=true}
  n75 = Subtract(A=@n74, B=@n73)
  n76 = Add(A=@n74, B=@n73)
  n77 = Subtract(
   A=Add(A=MaterialFunctionCall(Alpha=@n72, Min=@n75, Max=@n74) {materialFunction=&SmoothStep}.Result)
   B=Add(A=MaterialFunctionCall(Alpha=@n72, Min=@n74, Max=@n76) {materialFunction=&SmoothStep}.Result)
  )
  n78 = Multiply(A=Multiply(A=@n70, B=@n77) {constB=0.2}) {constB=0.2}
  n79 = Multiply(A=@n70, B=@n77)
  n80 = MaterialFunctionCall(Speed=Constant() {r=0.7}) {materialFunction=&DebugTimeSine}
  n81 = ComponentMask(None=CollectionParameter() {collection=&MPC_Minimap, parameterName=PlayerLocation}) {g=true, r=true}
  n82 = LinearInterpolate(A=Constant() {r=0.002}, B=Constant() {r=0.02}, Alpha=@n6)
  n83 = Add(
   A=Add(A=Add(A=@n67, B=@n69), B=LinearInterpolate(A=@n78, B=@n79, Alpha=@n80.Result) {constA=0.2})
   B=Multiply(A=Constant3Vector() {constant=<rgba 0 20 2 0>}, B=SphereMask(A=@n13.UVs, B=@n81, Radius=@n82))
  )
  n84 = MaterialFunctionCall!(RotateDialMask, Normal WS) {materialFunction=/Game/Packages/Fortress_SharedMaterials/Base_Material_Functions/AttachScope}
  n85 = Clamp(None=Multiply(A=Power(Base=@n84.Opacity) {constExponent=3}) {constB=20})
  n86 = SetMaterialAttributes!(
   MaterialAttributes
   Base Color=NamedRerouteDeclaration(None=Constant() {r=0.05}) {name=baseColor, nodeColor=<rgba 0.5703 0 1 1>}
   Metallic=NamedRerouteDeclaration(None=Constant() {r=1}) {name=metallic, nodeColor=<rgba 0 0.5078 1 1>}
   Roughness=NamedRerouteDeclaration(None=Multiply(A=@n34, B=@n36)) {name=roughness, nodeColor=<rgba 0 0.9531 1 1>}
   Emissive Color=NamedRerouteDeclaration(None=Multiply(A=@n35.R, B=@n83)) {name=emissive, nodeColor=<rgba 1 0.8203 0 1>}
   Normal=NamedRerouteDeclaration(None=@n84.World Normal) {name=nomral, nodeColor=<rgba 1 0 0.8203 1>}
   World Position Offset=NamedRerouteDeclaration(None=@n84.World Position Offset) {name=wpo, nodeColor=<rgba 0.2891 0 1 1>}
   Opacity Mask=NamedRerouteDeclaration(None=MaterialFunctionCall(Alpha Threshold=@n85) {materialFunction=&DitherTemporalAA}.Result) {name=opacityMask, nodeColor=<rgba 1 0 0.7969 1>}
  ) {attributeSetTypes [
    #1
    #2
    #3
    #4
    #5
    #6
    #7
   ]}
  MaterialFunctionCall(In=@n86) {materialFunction=&MF_TechArtOverrides_accessories}
  Multiply(A=@n35.R, B=@n39)
  Multiply(B=@n35.R)
 ]
}
